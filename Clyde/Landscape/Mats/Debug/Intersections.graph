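# Debug landscape material, unlit, additive and two-sided, that by its name serves to show intersections.
# TwoSidedSign becomes a 0/1 mask that picks negative red for front faces and positive red for back faces, with a Z parameter appended as blue.
# Negative emissive is allowed, so additive blending can subtract as well as add, letting overlapping opposite faces cancel.
# The graph shows no links to material outputs; a constant and vertex color feed nothing.
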Material /Clyde/Landscape/Mats/Debug/Intersections
graph {
 properties {
  bAllowNegativeEmissiveColor=true
  blendMode=BLEND_Additive
  pixelDepthOffsetMode=PDOM_Legacy
  shadingModel=MSM_Unlit
  twoSided=true
  usedShadingModels=Unlit
 }
 nodes [
  n1 = Constant3Vector() {constant=<rgba 1 0 0 0>}
  n2 = Constant3Vector() {constant=<rgba -1 0 0 0>}
  n3 = Saturate(None=Ceil(None=TwoSidedSign()))
  AppendVector(A=ComponentMask(None=LinearInterpolate(A=@n1, B=@n2, Alpha=@n3)) {g=true, r=true}, B=ScalarParameter() {parameterName=Z})
  Constant() {r=1}
  VertexColor()
 ]
}
